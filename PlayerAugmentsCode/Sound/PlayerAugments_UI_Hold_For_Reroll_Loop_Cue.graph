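# Sound cue for the player augments UI, named as the loop heard while holding to reroll.
# A mixer layers a hover tick, pitched up by a fixed modulator and mixed quieter, with a louder reroll hold charge wave.
# The cue uses the non-critical in-game UI sound class with virtualization disabled.
# Its concurrency override caps simultaneous instances at a small count, stopping the oldest first.
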SoundCue /PlayerAugmentsCode/Sound/PlayerAugments_UI_Hold_For_Reroll_Loop_Cue
root=Mixer_0
nodes [
 {id=Mixer_0, children [Modulator_1,WavePlayer_5], properties {inputVolume [0.5,1.2]}}
 {id=Modulator_1, children [WavePlayer_4], properties {pitchMax=1.5, pitchMin=1.5, volumeMax=1, volumeMin=1}}
 {id=WavePlayer_4, properties {soundWaveAssetPtr=&/PlayerAugments_UI_Hover_Tick_Center_01}}
 {id=WavePlayer_5, properties {soundWaveAssetPtr=&/PlayerAugments_UI_Reroll_Hold_Charge_01}}
]
properties {
 bOverrideConcurrency=true
 maxDistance=2097152
 soundClassObject=&Fort_UI_InGame_NonCritical
 virtualizationMode=Disabled
 volumeMultiplier=1.85
 concurrencyOverrides {maxCount=2, resolutionRule=StopOldest}
}
