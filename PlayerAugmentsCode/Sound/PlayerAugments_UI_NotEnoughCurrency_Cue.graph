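# UI sound cue for Player Augments, named for the not-enough-currency feedback.
# A single wave player feeds the root directly with the reroll-fail sound wave, with no randomization or modulation.
# It uses the non-critical in-game UI sound class.
# A concurrency override caps simultaneous instances, rejects new plays over the limit and adds a short retrigger time to curb spamming.
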SoundCue /PlayerAugmentsCode/Sound/PlayerAugments_UI_NotEnoughCurrency_Cue
root=WavePlayer_0
nodes [{id=WavePlayer_0, properties {soundWaveAssetPtr=&/PlayerAugments_UI_Reroll_Fail_01}}]
properties {
 bDebug=true
 bOverrideConcurrency=true
 maxDistance=2097152
 soundClassObject=&Fort_UI_InGame_NonCritical
 volumeMultiplier=1
 concurrencyOverrides {maxCount=2, resolutionRule=PreventNew, retriggerTime=0.2}
}
