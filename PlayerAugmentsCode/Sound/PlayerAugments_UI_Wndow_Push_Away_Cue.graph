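# PlayerAugments_UI_Wndow_Push_Away_Cue is a UI sound cue for pushing away a window in the player augments interface.
# A mixer layers two waves at equal volume, a HUD push-away select sound and an unequip sound, with slightly reduced overall volume.
# It plays in the non-critical in-game UI sound class with virtualization disabled.
# A concurrency override caps it at two simultaneous instances, stopping the oldest.
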SoundCue /PlayerAugmentsCode/Sound/PlayerAugments_UI_Wndow_Push_Away_Cue
root=Mixer_0
nodes [
 {id=Mixer_0, children [WavePlayer_1,WavePlayer_0], properties {inputVolume [1,1]}}
 {id=WavePlayer_0, properties {soundWaveAssetPtr=&/PlayerAugments_UI_Unequip_01}}
 {id=WavePlayer_1, properties {soundWaveAssetPtr=&/PlayerAugments_UI_Select_HUD_Push_Away_01}}
]
properties {
 bOverrideConcurrency=true
 maxDistance=2097152
 soundClassObject=&Fort_UI_InGame_NonCritical
 virtualizationMode=Disabled
 volumeMultiplier=0.8
 concurrencyOverrides {maxCount=2, resolutionRule=StopOldest}
}
